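# Two-sided lit material for a countdown monitor frame in Athena events, with a flat color passed through MF_CreativeMoveTool.
# World position offset hides the mesh by collapsing it to its pivot and dropping it far below, until MPC_Countdown's show value restores it.
# Static switches additionally require the phase 2 value or force visibility for debugging.
# A noise and vertex-color grime blend feeds a Fortnite_HitEffects call, separate from the plain base color.
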Material /BRCosmetics/Athena/Events/Generic/M_Monitor_Frame_Countdown
graph {
 properties {bUseMaterialAttributes=true, pixelDepthOffsetMode=PDOM_Legacy, twoSided=true, usedShadingModels=Default Lit}
 nodes [
  n1 = VectorParameter() {parameterName=Col_01, defaultValue=<rgba 0.04 0.04 0.04 1>}
  n2 = Add(A=Subtract(A=ObjectPositionWS(), B=WorldPosition().XYZ), B=Constant3Vector() {constant=<rgba 0 0 -5156156 0>})
  n3 = Constant3Vector()
  n4 = CollectionParameter() {collection=/Game/Athena/Events/Generic/CountdownTimer/Materials/MPC_Countdown, parameterName=ShowCountdown}
  n5 = CollectionParameter() {collection=&MPC_Countdown, parameterName=ShowCountdownPhase2}
  n6 = StaticSwitchParameter(True=Constant() {r=1}, False=StaticSwitchParameter(True=Multiply(A=@n4, B=@n5), False=@n4) {parameterName=isPhase2}) {parameterName=Debug}
  MaterialFunctionCall(MA In=MakeMaterialAttributes!(
    BaseColor=@n1.RGB
    Metallic
    Specular=ScalarParameter() {defaultValue=0.35, parameterName=Spec}
    Roughness=ScalarParameter() {defaultValue=0.5, parameterName=Rough}
    Anisotropy
    EmissiveColor
    Opacity
    OpacityMask
    Normal=TextureSample() {samplerType=SAMPLERTYPE_Normal, texture=&FlatNormalBlue}.RGB
    Tangent
    WorldPositionOffset=LinearInterpolate(A=@n2, B=@n3, Alpha=@n6)
    SubsurfaceColor
    ClearCoat
    ClearCoatRoughness
    AmbientOcclusion
    Refraction
    CustomizedUVs_0
    CustomizedUVs_1
    CustomizedUVs_2
    CustomizedUVs_3
    CustomizedUVs_4
    CustomizedUVs_5
    CustomizedUVs_6
    CustomizedUVs_7
    PixelDepthOffset
    ShadingModel
    Displacement
   )) {materialFunction=&MF_CreativeMoveTool}
  n7 = VectorParameter() {parameterName=Col_01_1, defaultValue=<rgba 0.0208 0.013 0.0087 1>}
  n8 = Multiply(A=@n1.RGB) {constB=0.75}
  n9 = Clamp(None=VertexColor().R)
  n10 = Rotator(Coordinate=TextureCoordinate() {uTiling=6, vTiling=6}, Time=ScalarParameter() {defaultValue=0.4952, parameterName=Rotate})
  n11 = Multiply(
   A=Subtract(
    A=TextureSample(UVs=@n10) {texture=/Game/Global/Textures/Default/Noise/TankMoteNoise}.R
    B=ScalarParameter() {defaultValue=0.921, parameterName=Subtract}
   )
   B=ScalarParameter() {defaultValue=21.7978, parameterName=Multiply}
  ) {constB=64}
  MaterialFunctionCall!(
   WSDeformationCenter=VectorParameter() {bUseCustomPrimitiveData=true, group=z_code, parameterName=WSDeformationCenter, primitiveDataIndex=20}.RGB
   Code_DestructionAmount=ScalarParameter() {bUseCustomPrimitiveData=true, group=z_code, parameterName=Code_DestructionAmount, primitiveDataIndex=13}
   Code_UseUvForDamage
   Code_UseHitEffects=StaticBoolParameter() {defaultValue=true, group=z_code, parameterName=Code_UseHitEffects}
   Code_Emissive
   WSDamageRadius=ScalarParameter() {bUseCustomPrimitiveData=true, defaultValue=1, group=z_code, parameterName=WSDamageRadius, primitiveDataIndex=26}
   CrackDamageShader
   Diffuse=LinearInterpolate(A=@n1.RGB, B=LinearInterpolate(A=@n7.RGB, B=@n8, Alpha=@n9), Alpha=Clamp(None=@n11))
   ScalarModDamageCracks
   ScalarOffsetforDamageCracks
   ExposeLerpFunctionForInvertingTheHitDistanceCalc
   InvertDamageFalloff
   CrackWorldPosition
   WorldPosition
   DistanceCullFadeAnimationOffset=ScalarParameter() {bUseCustomPrimitiveData=true, group=z_code, parameterName=DistanceCullFadeAnimationOffset, primitiveDataIndex=12}
  ) {materialFunction=/Game/Packages/jl_tests/Fortnite_HitEffects}
 ]
}
